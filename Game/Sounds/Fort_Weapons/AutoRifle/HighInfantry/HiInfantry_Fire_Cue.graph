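# Sound cue for the HighInfantry auto rifle's gunshot, layered by distance.
# A mixer combines a close layer in the third-person fire sound class with medium and distant layers in the distant fire class.
# Each layer picks randomly from three waves through its own projectile attenuation; the close picks are weighted, and the farther layers get a volume boost.
# Owner-limited concurrency stops the oldest shot, and local and global gunfire concurrency sets are assigned.
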SoundCue /Game/Sounds/Fort_Weapons/AutoRifle/HighInfantry/HiInfantry_Fire_Cue
root=Mixer_1
nodes [
 {id=Attenuation_118, children [Random_0], properties {attenuationSettings=&Fort_Weapon_Projectile_Close_Athena_Att}}
 {
  id=Attenuation_119
  children [Modulator_0]
  properties {attenuationSettings=&Fort_Weapon_Projectile_Medium_Mono_Athena_Att}
 }
 {
  id=Attenuation_120
  children [Modulator_1]
  properties {attenuationSettings=&Fort_Weapon_Projectile_Distant_Mono_Athena_Att}
 }
 {id=Mixer_1, children [SoundClass_77, SoundClass_76], properties {inputVolume [0.9,0.7]}}
 {id=Mixer_4, children [Attenuation_119, Attenuation_120], properties {inputVolume [0.8,0.8]}}
 {id=Modulator_0, children [Random_1], properties {volumeMax=1.7, volumeMin=1.7}}
 {id=Modulator_1, children [Random_2], properties {volumeMax=1.8, volumeMin=1.8}}
 {id=Random_0, children [WavePlayer_0,WavePlayer_2,WavePlayer_3], properties {weights [3,1,1]}}
 {id=Random_1, children [WavePlayer_4,WavePlayer_5,WavePlayer_6], properties {weights [1,1,1]}}
 {id=Random_2, children [WavePlayer_7,WavePlayer_8,WavePlayer_9], properties {weights [1,1,1]}}
 {id=SoundClass_76, children [Mixer_4], properties {soundClassOverride=&Fort_WeaponRanged_Fire_Distant}}
 {
  id=SoundClass_77
  children [Attenuation_118]
  properties {soundClassOverride=&Fort_WeaponRanged_Fire_3P_PSM}
 }
 {id=WavePlayer_0, properties {soundWaveAssetPtr=&/HiInfantry_Fire_Close_01}}
 {id=WavePlayer_2, properties {soundWaveAssetPtr=&/HiInfantry_Fire_Close_03}}
 {id=WavePlayer_3, properties {soundWaveAssetPtr=&/HiInfantry_Fire_Close_04}}
 {id=WavePlayer_4, properties {soundWaveAssetPtr=&/HiInfantry_Fire_Medium_01}}
 {id=WavePlayer_5, properties {soundWaveAssetPtr=&/HiInfantry_Fire_Medium_02}}
 {id=WavePlayer_6, properties {soundWaveAssetPtr=&/HiInfantry_Fire_Medium_03}}
 {id=WavePlayer_7, properties {soundWaveAssetPtr=&/HiInfantry_Fire_Distant_01}}
 {id=WavePlayer_8, properties {soundWaveAssetPtr=&/HiInfantry_Fire_Distant_02}}
 {id=WavePlayer_9, properties {soundWaveAssetPtr=&/HiInfantry_Fire_Distant_03}}
]
properties {
 bDebug=true
 maxDistance=27500
 soundClassObject=&Fort_WeaponRanged_Fire_3P_PSM
 volumeMultiplier=0.45
 concurrencyOverrides {bLimitToOwner=true, maxCount=3, resolutionRule=StopOldest}
 concurrencySet [&Weapon_Gunfire_Local_Conc, &Weapon_Gunfire_Global_Conc]
}
